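# Sound cue for IO elevator (IOSlider) announcements: a klaxon with guard voice lines, played as attenuated ambient audio.
# After an initial delay, a mixer plays the louder klaxon plus two quieter, staggered delay branches sharing one quality-level node.
# That node picks randomly from six male and female guard lines at the first two quality levels, and from a three-line subset at the last two.
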SoundCue /Lager/Sounds/SFX/Elevator/IOSlider_Announcements_Cue
root=Delay_4
nodes [
 {id=Delay_2, children [QualityLevel_0], properties {delayMax=1.5, delayMin=1.5}}
 {id=Delay_3, children [QualityLevel_0], properties {delayMax=3, delayMin=3}}
 {id=Delay_4, children [Mixer_0], properties {delayMax=1.5, delayMin=1.5}}
 {id=Mixer_0, children [WavePlayer_1,Delay_2,Delay_3], properties {inputVolume [1.3,0.3,0.3]}}
 {id=QualityLevel_0, children [Random_0,Random_0,Random_1,Random_1]}
 {
  id=Random_0
  children [WavePlayer_3, WavePlayer_4, WavePlayer_9, WavePlayer_10, WavePlayer_16, WavePlayer_17]
  properties {weights [1,1,1,1,1,1]}
 }
 {id=Random_1, children [WavePlayer_0,WavePlayer_2,WavePlayer_5], properties {bShouldExcludeFromBranchCulling=true, weights [1,1,1]}}
 {id=WavePlayer_0, properties {soundWaveAssetPtr=/Lager/Sounds/SFX/Elevator/IO_Guard_VO/IO_Guard_Female_Announcement_013}}
 {id=WavePlayer_1, properties {soundWaveAssetPtr=&/IO_Elevator_Klaxon_Announcement}}
 {id=WavePlayer_10, properties {soundWaveAssetPtr=/Lager/Sounds/SFX/Elevator/IO_Guard_VO/IO_Guard_Male_01_Announcement_015}}
 {id=WavePlayer_16, properties {soundWaveAssetPtr=/Lager/Sounds/SFX/Elevator/IO_Guard_VO/IO_Guard_Male_02_Announcement_018}}
 {id=WavePlayer_17, properties {soundWaveAssetPtr=/Lager/Sounds/SFX/Elevator/IO_Guard_VO/IO_Guard_Male_02_Announcement_027}}
 {id=WavePlayer_2, properties {soundWaveAssetPtr=/Lager/Sounds/SFX/Elevator/IO_Guard_VO/IO_Guard_Male_01_Announcement_014}}
 {id=WavePlayer_3, properties {soundWaveAssetPtr=&IO_Guard_Female_Announcement_013}}
 {id=WavePlayer_4, properties {soundWaveAssetPtr=/Lager/Sounds/SFX/Elevator/IO_Guard_VO/IO_Guard_Female_Announcement_014}}
 {id=WavePlayer_5, properties {soundWaveAssetPtr=&IO_Guard_Male_02_Announcement_018}}
 {id=WavePlayer_9, properties {soundWaveAssetPtr=&IO_Guard_Male_01_Announcement_014}}
]
properties {
 attenuationSettings=&/IOSlider_Doors_Att
 maxDistance=8700
 soundClassObject=&Fort_Ambient_Athena
 volumeMultiplier=1.039
}
