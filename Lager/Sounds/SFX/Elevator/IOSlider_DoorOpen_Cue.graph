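# Sound cue for the opening sound of an IOSlider elevator door.
# A Delay node with a fixed half-second wait feeds a single Wave Player for the door-open wave, so playback starts slightly after the cue is triggered.
# The cue uses the IOSlider doors attenuation asset, routes to the Athena ambient sound class and plays at reduced volume.
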SoundCue /Lager/Sounds/SFX/Elevator/IOSlider_DoorOpen_Cue
root=Delay_0
nodes [
 {id=Delay_0, children [WavePlayer_0], properties {delayMax=0.5, delayMin=0.5}}
 {id=WavePlayer_0, properties {soundWaveAssetPtr=&/IOSlider_DoorOpen_01}}
]
properties {
 attenuationSettings=&/IOSlider_Doors_Att
 maxDistance=8700
 soundClassObject=&Fort_Ambient_Athena
 volumeMultiplier=0.624
}
