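# IOSlider_Raise_Cue is a spatialized ambient sound cue for an IO Slider elevator raising.
# Its root mixer layers an earthquake shake blended with a Storm King weak-point destruction sound, shaped by a volume envelope that fades out over about three seconds.
# A slightly pitched-down, briefly delayed mechanical raise layer joins them, and an arrival sound follows after about a second, with each layer weighted separately.
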SoundCue /Lager/Sounds/SFX/Elevator/IOSlider_Raise_Cue
root=Mixer_1
nodes [
 {id=Delay_0, children [WavePlayer_8], properties {delayMax=1.1, delayMin=1.1}}
 {id=Delay_1, children [WavePlayer_0], properties {delayMax=0.2, delayMin=0.2}}
 {
  id=Enveloper_0
  children [Mixer_0]
  properties {
   pitchCurve {editorCurveData {keys [{time=0}]}}
   volumeCurve {editorCurveData {keys [<at 0 1>, {arriveTangent=-0.3333, interpMode=RCIM_Cubic, leaveTangent=-1.1858, tangentMode=RCTM_Break, time=0.3964, value=0.9978}, <at 3 0>]}}
  }
 }
 {id=Mixer_0, children [WavePlayer_3,WavePlayer_7], properties {inputVolume [1,1]}}
 {id=Mixer_1, children [Enveloper_0,Modulator_0,Delay_0], properties {inputVolume [0.7,1.3,0.5]}}
 {id=Modulator_0, children [Delay_1], properties {pitchMax=0.9, pitchMin=0.9, volumeMax=1, volumeMin=1}}
 {id=WavePlayer_0, properties {soundWaveAssetPtr=&/IOSlider_Elevator_RaisedMechanical_01}}
 {id=WavePlayer_3, properties {soundWaveAssetPtr=/Game/Sounds/Fort_Impact_Sounds/Demolition/Earthquake/EarthQuake_Shakes_Small}}
 {id=WavePlayer_7, properties {soundWaveAssetPtr=/Game/Sounds/Fort_Monster_Storm_King/SFX/Final_SFX/StormKing_WeakPoint_Destroyed_v2_02}}
 {id=WavePlayer_8, properties {soundWaveAssetPtr=&/IOSlider_Elevator_Arrived_01}}
]
properties {
 attenuationSettings=&/IOSlider_Att
 maxDistance=9700
 soundClassObject=&Fort_Ambient_Athena
 volumeMultiplier=1.351
}
