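# Sound cue for the IO Slider elevator's pre-tell rumble, warning of movement with a shake layered over a building motor-like loop.
# A Mixer blends a one-shot ground-and-metal shake, weighted louder, with a quieter looping elevator ascend sound.
# An Enveloper on the loop fades it in over a few seconds and steadily raises its pitch over about 24 seconds.
# It uses Slider attenuation, the Athena ambient sound class, and keeps playing when silent.
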SoundCue /Lager/Sounds/SFX/Elevator/IOSlider_Rumble_PreTell_Cue
root=Mixer_0
nodes [
 {
  id=Enveloper_2
  children [WavePlayer_3]
  properties {
   pitchCurve {editorCurveData {keys [<at 0.0148 0.7179>,<at 24.3802 3.0343>]}}
   volumeCurve {editorCurveData {keys [<at 0.1239 0.0314>,<at 1.3709 0.843>,<at 3.3212 0.999>]}}
  }
 }
 {id=Mixer_0, children [WavePlayer_0,Enveloper_2], properties {inputVolume [1.3,0.3]}}
 {id=WavePlayer_0, properties {soundWaveAssetPtr=&/IO_Elevator_GroundAndMetal_Shake_01}}
 {id=WavePlayer_3, properties {bLooping=true, soundWaveAssetPtr=&/IO_Elevator_Ascend_Loop_Mono_01}}
]
properties {
 attenuationSettings=&/IOSlider_Att
 maxDistance=9700
 soundClassObject=&Fort_Ambient_Athena
 virtualizationMode=PlayWhenSilent
 volumeMultiplier=1.351
}
